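# Translucent unlit particle material for the ShatterFly glider's aura effect, enabled for Niagara and Cascade sprites.
# Two smoke-texture layers pan on dynamic-parameter time, distorted by macro-UV noise, then are multiplied and shaped by a power.
# Particle color tints the result, which passes through LDR-to-HDR remapping driven by a dynamic Emissive parameter and shared emissive adjustments.
# Opacity is particle alpha times a squared radial sphere mask, giving soft circular edges.
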Material /Game/Gadgets/Assets/VinderTech_GliderChute/Glider_ShatterFly/FX/M_Aura
graph {
 properties {
  bUsedWithNiagaraSprites=true
  bUsedWithParticleSprites=true
  blendMode=BLEND_Translucent
  pixelDepthOffsetMode=PDOM_Legacy
  shadingModel=MSM_Unlit
  usedShadingModels=Unlit
 }
 nodes [
  n1 = TextureCoordinate()
  n2 = VectorParameter() {parameterName=UVS1, defaultValue=<rgba 1 0.5 -0.1 0.3>}
  n3 = Multiply(A=@n1, B=ComponentMask(None=@n2.RGB) {g=true, r=true})
  n4 = DynamicParameter() {defaultValue=<rgba 1 1 25 25>, paramNames [Time,Random,Param3,Emissive]}
  n5 = Add(A=@n4.Time, B=@n4.Time)
  n6 = AppendVector(A=@n2.B, B=@n2.B)
  n7 = TextureSample(UVs=Multiply(A=ParticleMacroUV(), B=ScalarParameter() {defaultValue=2, parameterName=MacroUV})) {samplerType=SAMPLERTYPE_LinearColor, texture=&T_NoisePack_LinearColor}
  n8 = Multiply(A=Subtract(A=ComponentMask(None=@n7.RGB) {g=true, r=true}) {constB=0.33}, B=ScalarParameter() {defaultValue=0.1, parameterName=Turb})
  n9 = ShadingPathSwitch(Default=Add(A=Panner(Coordinate=@n3, Time=@n5, Speed=@n6), B=@n8)) {desc=Migrated from FeatureLevelSwitch}
  n10 = VectorParameter() {parameterName=UVS2, defaultValue=<rgba 0.8 0.8 0.1 0.1>}
  n11 = Multiply(A=@n1, B=ComponentMask(None=@n10.RGB) {g=true, r=true})
  n12 = AppendVector(A=@n10.B, B=@n10.B)
  n13 = ShadingPathSwitch(Default=Add(A=Panner(Coordinate=@n11, Time=@n5, Speed=@n12), B=@n8)) {desc=Migrated from FeatureLevelSwitch}
  n14 = Power(
   Base=Multiply(
    A=TextureSample(UVs=@n9) {texture=&T_Inky_Smoke_Tile_Chromatic}.RGB
    B=TextureSample(UVs=@n13) {texture=&T_Inky_Smoke_Tile_Chromatic}.RGB
   )
   Exp=ScalarParameter() {defaultValue=0.7, parameterName=Exp}
  )
  n15 = ParticleColor()
  MaterialFunctionCall(Emissive=MaterialFunctionCall(
    LDR Input=Multiply(A=@n14, B=@n15.RGB)
    HDR Tint=@n15.RGB
    HDR Intensity Multiplier=DynamicParameter() {defaultValue=<rgba 1 1 25 25>, paramNames [Time,Random,Param3,Emissive]}.Emissive
   ) {materialFunction=&DeriveHDRfromLDR}.RemappedRange) {materialFunction=&MF_FX_EmissiveAdjustments}
  n16 = Constant() {r=0.5}
  n17 = SphereMask(A=TextureCoordinate(), B=@n16, Radius=@n16, Hardness=Constant())
  Multiply(A=@n15.A, B=Multiply(A=@n17, B=@n17))
 ]
}
